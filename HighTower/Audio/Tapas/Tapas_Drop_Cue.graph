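# Sound cue for dropping the Tapas item, likely a weapon given its weapon-drop attenuation.
# A single wave player feeds the Tapas drop sample through a Modulator left at default settings for slight playback variation.
# The cue plays at reduced volume under the loot-collect sound class, with shared weapon-drop attenuation handling spatial falloff.
# Its concurrency set holds only a null entry.
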SoundCue /HighTower/Audio/Tapas/Tapas_Drop_Cue
root=Modulator_1
nodes [{id=Modulator_1, children [WavePlayer_5]}, {id=WavePlayer_5, properties {soundWaveAssetPtr=&/Tapas_Drop_01}}]
properties {
 attenuationSettings=&Fort_Weapon_Drop_Att
 maxDistance=3000
 soundClassObject=&Fort_Loot_Collect
 volumeMultiplier=0.55
 concurrencySet [null]
}
