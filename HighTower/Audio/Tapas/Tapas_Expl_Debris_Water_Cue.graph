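# Sound cue for the water debris layer of a Tapas explosion, using Athena explosion debris waves.
# A random node picks one of four equally weighted waves, a short randomized delay offsets it, and a default modulator at the root varies pitch and volume.
# Playback uses close water-explosion attenuation, the large explosives sound class and the shared weapon explosion concurrency set.
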SoundCue /HighTower/Audio/Tapas/Tapas_Expl_Debris_Water_Cue
root=Modulator_2
nodes [
 {id=Delay_0, children [Random_0], properties {delayMax=0.13, delayMin=0.11}}
 {id=Modulator_2, children [Delay_0]}
 {id=Random_0, children [WavePlayer_2,WavePlayer_3,WavePlayer_5,WavePlayer_6], properties {weights [1,1,1,1]}}
 {id=WavePlayer_2, properties {soundWaveAssetPtr=/Game/Athena/Sounds/Weapons/Explosions/EXPL_Debris_Water_Wave_04}}
 {id=WavePlayer_3, properties {soundWaveAssetPtr=/Game/Athena/Sounds/Weapons/Explosions/EXPL_Debris_Water_Wave_01}}
 {id=WavePlayer_5, properties {soundWaveAssetPtr=/Game/Athena/Sounds/Weapons/Explosions/EXPL_Debris_Water_Wave_02}}
 {id=WavePlayer_6, properties {soundWaveAssetPtr=/Game/Athena/Sounds/Weapons/Explosions/EXPL_Debris_Water_Wave_03}}
]
properties {
 attenuationSettings=/Game/Sounds/Attenuation/Fort_Explosion_Water_Close_Athena_Att
 maxDistance=7241.2505
 soundClassObject=/Game/Sounds/FortSoundClassesAndModes/Fort_Explosives_Large_PSM
 volumeMultiplier=0.55
 concurrencyOverrides {maxCount=4, resolutionRule=StopOldest, volumeScale=0.5}
 concurrencySet [/Game/Sounds/Concurrency_Rules/Weapon_Explosion_Conc]
}
