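# Sound cue for the Tapas pickup, classed as a loot-collect sound.
# A single wave player feeds Tapas_Pickup_01 through a Modulator node left at its default settings, with the cue lowering overall volume and pitch.
# It borrows a close-range umbrella weapon attenuation asset for spatial falloff.
# A concurrency override allows one instance at a time and blocks new plays while one is active.
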SoundCue /HighTower/Audio/Tapas/Tapas_Pickup_Cue
root=Modulator_0
nodes [{id=Modulator_0, children [WavePlayer_1]}, {id=WavePlayer_1, properties {soundWaveAssetPtr=&/Tapas_Pickup_01}}]
properties {
 attenuationSettings=/Game/Sounds/Attenuation/Fort_Weapon_Umbrella_Close_Athena_Att
 bOverrideConcurrency=true
 maxDistance=4500
 pitchMultiplier=0.9
 soundClassObject=&Fort_Loot_Collect
 volumeMultiplier=0.3
 concurrencyOverrides {maxCount=1, resolutionRule=PreventNew}
}
