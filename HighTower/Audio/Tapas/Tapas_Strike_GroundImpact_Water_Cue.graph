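# Sound cue for the Tapas strike's ground impact on water, layering lightning strike, close thunder and water splash in the melee weapon sound class.
# A quality switch sends its first slot to a mix of randomized variations per layer and its other three to a shared mix with one fixed wave per layer.
# Both mixes get slight random pitch and volume modulation.
# Overrides add HRTF spatialization, distance low-pass attenuation and a concurrency cap.
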SoundCue /HighTower/Audio/Tapas/Tapas_Strike_GroundImpact_Water_Cue
root=QualityLevel_1
nodes [
 {id=Mixer_1, children [Random_4,Random_5,Random_2], properties {inputVolume [0.9,0.7,1.2]}}
 {id=Mixer_2, children [Random_7,Random_8,WavePlayer_9], properties {inputVolume [0.9,0.7,1]}}
 {id=Modulator_0, children [Mixer_2], properties {pitchMax=1, pitchMin=0.9, volumeMax=1.1, volumeMin=0.9}}
 {id=Modulator_1, children [Mixer_1], properties {pitchMax=1, pitchMin=0.9, volumeMax=1.1, volumeMin=0.9}}
 {id=QualityLevel_1, children [Modulator_1,Modulator_0,Modulator_0,Modulator_0]}
 {id=Random_2, children [WavePlayer_6,WavePlayer_7,WavePlayer_8], properties {weights [1,1,1]}}
 {id=Random_4, children [WavePlayer_12, WavePlayer_13], properties {weights [1,1]}}
 {id=Random_5, children [WavePlayer_14, WavePlayer_15, WavePlayer_16], properties {weights [1,1,1]}}
 {id=Random_7, children [WavePlayer_20], properties {weights [1]}}
 {id=Random_8, children [WavePlayer_23], properties {weights [1]}}
 {id=WavePlayer_12, properties {soundWaveAssetPtr=&/Tapas_Strike_LightningStrike_02}}
 {id=WavePlayer_13, properties {soundWaveAssetPtr=&/Tapas_Strike_LightningStrike_03}}
 {id=WavePlayer_14, properties {soundWaveAssetPtr=&/Thunder_Close_01}}
 {id=WavePlayer_15, properties {soundWaveAssetPtr=&/Thunder_Close_02}}
 {id=WavePlayer_16, properties {soundWaveAssetPtr=&/Thunder_Close_03}}
 {id=WavePlayer_20, properties {soundWaveAssetPtr=&Tapas_Strike_LightningStrike_02}}
 {id=WavePlayer_23, properties {soundWaveAssetPtr=&Thunder_Close_02}}
 {id=WavePlayer_6, properties {soundWaveAssetPtr=&/EXPL_Debris_Water_Splash_Med_2}}
 {id=WavePlayer_7, properties {soundWaveAssetPtr=&/EXPL_Debris_Water_Splash_Med_3}}
 {id=WavePlayer_8, properties {soundWaveAssetPtr=&/EXPL_Debris_Water_Splash_Med_4}}
 {id=WavePlayer_9, properties {soundWaveAssetPtr=&EXPL_Debris_Water_Splash_Med_2}}
]
properties {
 attenuationSettings=/Game/Sounds/Attenuation/Fort_Wasabi_Melee_Att
 bOverrideAttenuation=true
 bOverrideConcurrency=true
 maxDistance=5400
 soundClassObject=&Fort_WeaponMelee
 volumeMultiplier=0.4
 attenuationOverrides {
  bAttenuateWithLPF=true
  falloffDistance=3700
  lPFFrequencyAtMax=200
  lPFRadiusMax=4000
  lPFRadiusMin=2000
  spatializationAlgorithm=SPATIALIZATION_HRTF
  stereoSpread=2000
  attenuationShapeExtents {x=1700}
 }
 concurrencyOverrides {maxCount=12}
 concurrencySet [null]
}
