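# Sound cue for the Tapas strike pretell, the audio tell before a strike, in the melee weapon sound class.
# A mixer layers three waves at once: a briefly delayed, pitched-down hand raise, a start sound, and a slightly quieter loop wave.
# No looping node is present in the cue.
# Inline overrides supply HRTF-spatialized attenuation and a concurrency limit.
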SoundCue /HighTower/Audio/Tapas/Tapas_Strike_Pretell_Loop_Cue
root=Mixer_0
nodes [
 {id=Delay_0, children [WavePlayer_0], properties {delayMax=0.04, delayMin=0.04}}
 {id=Mixer_0, children [Modulator_0,WavePlayer_1,WavePlayer_3], properties {inputVolume [1,1,0.8]}}
 {id=Modulator_0, children [Delay_0], properties {pitchMax=0.9, pitchMin=0.9}}
 {id=WavePlayer_0, properties {soundWaveAssetPtr=&/Tapas_Strike_Pretell_HandRaise_01}}
 {id=WavePlayer_1, properties {soundWaveAssetPtr=&/Tapas_Strike_Pretell_Start_01}}
 {id=WavePlayer_3, properties {soundWaveAssetPtr=&/Tapas_Strike_Pretell_Loop_01}}
]
properties {
 attenuationSettings=/Game/Sounds/Attenuation/Fort_Wasabi_Melee_Att
 bOverrideAttenuation=true
 bOverrideConcurrency=true
 maxDistance=7700
 soundClassObject=&Fort_WeaponMelee
 volumeMultiplier=0.54
 attenuationOverrides {falloffDistance=5000, spatializationAlgorithm=SPATIALIZATION_HRTF, stereoSpread=3000, attenuationShapeExtents {x=2700}}
 concurrencyOverrides {maxCount=12}
 concurrencySet [null]
}
