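# Sound cue for the Tapas strike throw, a weapon-melee class sound layering throw, sky storm, lightning beam and build waves.
# Each mixer combines a briefly delayed throw, later storm and beam layers and an undelayed build, then feeds slight random pitch and volume modulation.
# The first quality level randomizes between wave variants, while the other three share a single-wave-per-layer branch.
# Cue-level overrides set HRTF spatialized attenuation and limit concurrency.
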SoundCue /HighTower/Audio/Tapas/Tapas_Strike_Throw_Cue
root=QualityLevel_0
nodes [
 {id=Delay_0, children [Random_4], properties {delayMax=0.4, delayMin=0.4}}
 {id=Delay_1, children [Random_1], properties {delayMax=0.1, delayMin=0.1}}
 {id=Delay_4, children [Random_2], properties {delayMax=0.4, delayMin=0.4}}
 {id=Delay_5, children [Random_10], properties {delayMax=0.4, delayMin=0.4}}
 {id=Delay_6, children [Random_8], properties {delayMax=0.1, delayMin=0.1}}
 {id=Delay_7, children [Random_9], properties {delayMax=0.4, delayMin=0.4}}
 {id=Mixer_0, children [Delay_1, Delay_4, Delay_0, WavePlayer_21], properties {inputVolume [0.8,1.4,1.6,1.7]}}
 {id=Mixer_2, children [Delay_6, Delay_7, Delay_5, WavePlayer_29], properties {inputVolume [0.8,1.4,1.6,1.7]}}
 {id=Modulator_0, children [Mixer_2], properties {pitchMax=1, pitchMin=0.9, volumeMax=1.1, volumeMin=0.9}}
 {id=Modulator_1, children [Mixer_0], properties {pitchMax=1, pitchMin=0.9, volumeMax=1.1, volumeMin=0.9}}
 {id=QualityLevel_0, children [Modulator_1,Modulator_0,Modulator_0,Modulator_0]}
 {id=Random_1, children [WavePlayer_3,WavePlayer_4,WavePlayer_5], properties {weights [1,1,1]}}
 {id=Random_10, children [WavePlayer_27], properties {weights [1]}}
 {id=Random_2, children [WavePlayer_7,WavePlayer_8], properties {weights [1,1]}}
 {id=Random_4, children [WavePlayer_12, WavePlayer_13], properties {weights [1,1]}}
 {id=Random_8, children [WavePlayer_22], properties {weights [1]}}
 {id=Random_9, children [WavePlayer_25], properties {weights [1]}}
 {id=WavePlayer_12, properties {soundWaveAssetPtr=&/Tapas_Strike_LightningBeam_01}}
 {id=WavePlayer_13, properties {soundWaveAssetPtr=&/Tapas_Strike_LightningBeam_02}}
 {id=WavePlayer_21, properties {soundWaveAssetPtr=&/Tapas_Strike_Build_01}}
 {id=WavePlayer_22, properties {soundWaveAssetPtr=&/Tapas_Strike_Throw_01}}
 {id=WavePlayer_25, properties {soundWaveAssetPtr=&/Tapas_Strike_SkyStorm_02}}
 {id=WavePlayer_27, properties {soundWaveAssetPtr=&Tapas_Strike_LightningBeam_01}}
 {id=WavePlayer_29, properties {soundWaveAssetPtr=&Tapas_Strike_Build_01}}
 {id=WavePlayer_3, properties {soundWaveAssetPtr=&Tapas_Strike_Throw_01}}
 {id=WavePlayer_4, properties {soundWaveAssetPtr=&/Tapas_Strike_Throw_02}}
 {id=WavePlayer_5, properties {soundWaveAssetPtr=&/Tapas_Strike_Throw_03}}
 {id=WavePlayer_7, properties {soundWaveAssetPtr=&Tapas_Strike_SkyStorm_02}}
 {id=WavePlayer_8, properties {soundWaveAssetPtr=&/Tapas_Strike_SkyStorm_03}}
]
properties {
 attenuationSettings=/Game/Sounds/Attenuation/Fort_Wasabi_Melee_Att
 bOverrideAttenuation=true
 bOverrideConcurrency=true
 maxDistance=7700
 soundClassObject=&Fort_WeaponMelee
 volumeMultiplier=0.4
 attenuationOverrides {falloffDistance=5000, spatializationAlgorithm=SPATIALIZATION_HRTF, stereoSpread=3000, attenuationShapeExtents {x=2700}}
 concurrencyOverrides {maxCount=12}
 concurrencySet [null]
}
